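# Translucent two-sided smoke material for Niagara mesh particles, likely the mask smoke on the CrispRover character cosmetic.
# World position offset combines bending driven by particle dynamic parameters with scrolling cloud noise, both weighted along U.
# Emissive blends base and tip colors tinted by particle color.
# Opacity erodes a panning smoke texture along U, adds noise-varied depth fade except on mobile, and clears a sphere around the instance-space head position.
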Material /BRCosmetics/Characters/Player/Male/Medium/Bodies/M_MED_CrispRover_BreezeSea/FX/M_CrispRoverMaskSmoke
graph {
 properties {
  bUsedWithNiagaraMeshParticles=true
  blendMode=BLEND_Translucent
  pixelDepthOffsetMode=PDOM_Legacy
  twoSided=true
  usedShadingModels=Default Lit
 }
 nodes [
  ActorPositionWS()
  n1 = DynamicParameter() {parameterIndex=1, defaultValue=<rgba 0 1 1 1>, paramNames [SpeedX,SpeedY,SpeedZ,Param4]}
  n2 = ComponentMask(None=TextureCoordinate()) {r=true}
  n3 = ScalarParameter() {defaultValue=3, parameterName=BendingPower}
  n4 = TextureCoordinate()
  n5 = VectorParameter() {parameterName=WPONoiseUv, channelNames {a=PanY, b=PanX, g=ScaleY, r=ScaleX}, defaultValue=<rgba 0.1 0 -0.05 0>}
  n6 = AppendVector(A=@n5.ScaleX, B=@n5.ScaleX)
  n7 = Multiply(A=AppendVector(A=@n5.PanX, B=@n5.PanX), B=Time())
  n8 = DynamicParameter() {defaultValue=<rgba 0 1 1 1>, paramNames [Erode,UVOffset,Param3,Param4]}
  n9 = TextureSample(UVs=Add(A=Multiply(A=@n4, B=@n6), B=Add(A=@n7, B=@n8.UVOffset))) {texture=/Game/TimeOfDay/LightFunction/T_Cloud2dNoise}
  n10 = ScalarParameter() {defaultValue=3, parameterName=WPOStrength}
  n11 = ComponentMask(None=TextureCoordinate()) {r=true}
  Add(
   A=Multiply(
    A=MaterialFunctionCall(X=@n1.SpeedX, Y=@n1.SpeedX, Z=@n1.SpeedX) {materialFunction=&MakeFloat3}.Result
    B=Power(Base=@n2, Exp=@n3) {constExponent=2.2}
   )
   B=Multiply(A=ConstantBiasScale(None=@n9.RGB) {bias=-0.5, scale=2}, B=Multiply(A=@n10, B=@n11))
  )
  n12 = VectorParameter() {parameterName=ColorBase, defaultValue=<rgba 0.1072 0.0075 0.5 1>}
  n13 = VectorParameter() {parameterName=ColorTip, defaultValue=<rgba 0.3915 0.0306 0.474 1>}
  n14 = TextureCoordinate()
  n15 = SmoothStep(Value=Power(Base=ComponentMask(None=@n14) {r=true}))
  n16 = ParticleColor()
  MaterialFunctionCall(Emissive=Multiply(A=LinearInterpolate(A=@n12.RGB, B=@n13.RGB, Alpha=@n15), B=@n16.RGB)) {materialFunction=&MF_FX_EmissiveAdjustments}
  n17 = TextureCoordinate()
  n18 = VectorParameter() {parameterName=SmokeRibbonUv, channelNames {a=PanY, b=PanX, g=ScaleY, r=ScaleX}, defaultValue=<rgba 1 1 -0.5 0>}
  n19 = AppendVector(A=@n18.ScaleX, B=@n18.ScaleX)
  n20 = Multiply(A=AppendVector(A=@n18.PanX, B=@n18.PanX), B=Time())
  n21 = MaterialFunctionCall(X=DynamicParameter() {defaultValue=<rgba 0 1 1 1>, paramNames [Erode,UVOffset,Param3,Param4]}.UVOffset) {materialFunction=&MakeFloat2}
  n22 = TextureSample(UVs=Add(A=Multiply(A=@n17, B=@n19), B=Add(A=@n20, B=@n21.Result))) {samplerType=SAMPLERTYPE_Grayscale, texture=/Game/Effects/Fort_Effects/Textures/Smoke/T_Smoke_Trail_Soft_NoClamp}
  n23 = Add(A=@n22.R)
  n24 = Multiply(A=ComponentMask(None=TextureCoordinate()) {r=true}) {constB=2}
  n25 = Multiply(A=Multiply(A=@n16.A, B=SmoothStep(Value=@n22.R)), B=Saturate(None=Subtract(A=@n23, B=@n24)))
  n26 = OneMinus(None=@n22.R)
  n27 = ScalarParameter() {defaultValue=5, parameterName=FadeDist}
  n28 = Multiply(A=@n25, B=DepthFade(FadeDistance=Multiply(A=@n26, B=@n27))) {constA=1}
  n29 = ComponentMask(None=TextureCoordinate()) {r=true}
  n30 = VectorParameter() {parameterName=HeadPos, defaultValue=<rgba 40 0 0 0>}
  n31 = Length(None=Subtract(
    A=WorldPosition().XYZ
    B=TransformPosition(None=@n30.RGB) {transformSourceType=TRANSFORMPOSSOURCE_Instance, transformType=TRANSFORMPOSSOURCE_World}
   )) {is=class MaterialExpressionLength}
  n32 = ScalarParameter() {defaultValue=40, parameterName=HeadRadius}
  Multiply(
   A=Multiply(
    A=ShadingPathSwitch(Default=@n28, Deferred=@n28, Mobile=@n25) {desc=Migrated from FeatureLevelSwitch}
    B=SmoothStep(Value=@n29) {constMax=0.1}
   )
   B=Saturate(None=Subtract(A=@n31, B=@n32) {constB=25})
  )
 ]
}
