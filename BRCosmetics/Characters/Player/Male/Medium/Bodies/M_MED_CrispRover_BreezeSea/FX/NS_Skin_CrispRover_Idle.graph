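# Idle cosmetic Niagara effect for the CrispRover player skin, drawing smoke trails as mesh particles attached to the character's skeletal mesh.
# One emitter bursts two persistent meshes when SmokeOffOn is set, using spawn group to pick each socket and mirror scale.
# A local BendSmokeMesh module bends them by sampled socket velocity, with separate tuning for skydiving.
# Scale and alpha follow SmokeOffOn and VariantIndex, so only the third variant shows smoke.
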NiagaraSystem /BRCosmetics/Characters/Player/Male/Medium/Bodies/M_MED_CrispRover_BreezeSea/FX/NS_Skin_CrispRover_Idle
dependencies {
 usedDataInterfaces [&NiagaraDataInterfaceArrayInt32, &NiagaraDataInterfaceArrayFloat]
 usedDynamicInputs [&MaskFloatBySpawnGroup, &FNTimeOfDayScalar]
 usedModules [
  &SystemState
  &self:SystemSpawnScript.SystemScriptSource.SystemScriptGraph.NiagaraNodeAssignment_2.SetVariables_#1
 ]
}
emitters {SmokeTrails001 {
  emitterUpdateScript [
   &EmitterState(
    Life Cycle Mode=::NewEnumerator1
    Loop Behavior=::NewEnumerator1
    Loop Delay=&Add_Float(A=$User.Delay, B=0.05)
    Loop Duration=1
    UseLoopDelay=true
   )
   &SpawnBurst_Instantaneous(
    Spawn Count=&SelectIntFromArray(
     Direct Array Index=&MakeIntFromFloat(Float=$User.SmokeOffOn)
     Int Selection Array=DataInterface() {propertyValues {IntData [0,2]}}
    )
    Use Spawn Probability=false
    UseLoopCountLimit=false
   )
  ]
  particleSpawnScript [
   &FN_SetSpawnGroup()
   &InitializeParticle(
    Color=<rgba 1 1 1 0.7>
    Mesh Scale=&MaskVectorBySpawnGroup(Masked Vector=<-1 1 1>, Passthrough Vector=<1 1 1>)
    Mesh Scale Mode=::NewEnumerator4
    UsePositionOffset=false
    Write Material Random=true
   )
   &SkeletalMeshLocation(
    Bone / Socket Index=&SpawnGroup()
    MaskSpawnGroup=false
    Mesh Sampling Type=::NewEnumerator5
    Skeletal Mesh=$User.SkeletalMesh
    Socket Sampling Mode=::NewEnumerator5
   )
   &DynamicMaterialParameters(
    Index 0 Param 2=&Multiply_Float(A=0.25, B=&ReturnNormalizedExecIndex())
    Module.Param0WriteEnabled=true
    Module.Param1WriteEnabled=true
    Module.Param2WriteEnabled=true
    Module.Param3WriteEnabled=true
   )
   &InitialMeshOrientation(Mesh Orientation Mode=::NewEnumerator7, Rotation=<0.35 0 0>, Rotation Coordinate Space=::NewEnumerator0, Use Rotation Vector=true)
  ]
  particleUpdateScript [
   &ParticleState(Kill Particles When Lifetime Has Elapsed=false)
   &SkeletalMeshLocation(
    Bone / Socket Index=&SpawnGroup()
    MaskSpawnGroup=false
    Mesh Sampling Type=::NewEnumerator5
    Sampled Velocity Coordinate Space:ENiagaraCoordinateSpace=::World
    Skeletal Mesh=$User.SkeletalMesh
    Socket Sampling Mode=::NewEnumerator5
   ) {module=SkeletalMeshLocation001}
   &SolveForcesAndVelocity(Clamp Velocity=false, Limit Acceleration=false)
   &self:BendSmokeMesh(
    BendLimitMax:Vector3f=&MakeCustomVectorFromBool(Bool=$User.IsSkydiving, True Vector=<10 10 25>, False Vector=<20 20 25>)
    BoneVelocity:Vector3f=$Particles.SkeletalMeshLocation001.SampledVelocity
    RestSpeed:NiagaraFloat=&MakeCustomFloatFromBool(Bool=$User.IsSkydiving, True Float=5, False Float=5)
    VelocityMultiplier:Vector3f=&MakeCustomVectorFromBool(Bool=$User.IsSkydiving, True Vector=<6 6 5>, False Vector=<10 10 5>)
   ) {module=BendSmokeMesh}
   &ScaleMeshSize(Scale Factor=&VectorFromFloat(Value=&Multiply_Float(
      A=$User.SmokeOffOn
      B=&SelectFloatFromArray(
       Direct Array Index=&MakeIntFromFloat(Float=$User.VariantIndex)
       Float Selection Array=DataInterface() {propertyValues {FloatData [0,0,1]}}
      )
     )))
   &ScaleColor(
    Scale Alpha=&Multiply_Float(
     A=$User.SmokeOffOn
     B=&SelectFloatFromArray(
      Direct Array Index=&MakeIntFromFloat(Float=$User.VariantIndex)
      Float Selection Array=DataInterface() {propertyValues {FloatData [0,0,1]}}
     )
    )
    ScaleA=true
    ScaleRGB=true
   )
  ]
  rendererClasses [&NiagaraMeshRendererProperties]
  renderers [{rendererClass=&NiagaraMeshRendererProperties, rendererIndex=0}]
 }}
summary {
 systemName=NS_Skin_CrispRover_Idle
 emitters [{emitterName=SmokeTrails001, rendererClasses [&NiagaraMeshRendererProperties]}]
 userVariables [
  User.bIsFrontEnd {type=NiagaraBool}
  User.bIsFrontEndPreview {type=NiagaraBool}
  User.Delay {type=NiagaraFloat, defaultValue=0, description=""}
  User.FlickerSpeed {type=NiagaraFloat, defaultValue=0, description=""}
  User.IsSkydiving {type=NiagaraBool}
  User.RibbonIDOffset {type=NiagaraInt32}
  User.ShapeSwitchSpeed {type=NiagaraFloat, defaultValue=0.5, description=""}
  User.SkeletalMesh {
   type=NiagaraDataInterfaceSkeletalMesh
   defaultValue {
    struct=NiagaraExt_VariableValue_DataInterface
    value {dataInterface=&self:NiagaraDataInterfaceSkeletalMesh_0}
   }
  }
  User.SmokeOffOn {type=NiagaraFloat, defaultValue=1, description=""}
  User.VariantIndex {type=NiagaraFloat, defaultValue=2, description=""}
 ]
}
