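# Animation Blueprint for the CrispRover BreezeSea male body cosmetic, adding secondary motion and pose correctives to its copied parent pose.
# A Control Rig feeds weighted constraints, look-ats and additive bone offsets for attachments, neck, arm, canisters and belt, then AnimDynamics swings loose accessories.
# Each update maps leg, neck and spine deltas from reference pose into corrective alphas.
# The right-arm aim alpha updates conditionally on crouching or skydive floating, resetting when either ends.
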AnimBlueprint /BRCosmetics/Characters/Player/Male/Medium/Bodies/M_MED_CrispRover_BreezeSea/Meshes/M_MED_CrispRover_BreezeSea_AnimBP
target_skeleton=&/M_MED_CrispRover_BreezeSea_Skeleton
graphs [
 AnimGraph {nodes [
   n1 = Control Rig(Source=Copy Pose From Mesh() {bUseAttachedParent=true}) {
    alphaBoolBlend {blendOption=Linear}
    controlRigAssetReference {blueprintRigClass=&/M_MED_CrispRover_BreezeSea_ControlRig.M_MED_CrispRover_BreezeSea_ControlRig_C}
   }
   n2 = "Constraint - attach_cape (PPPPPPP)"!(
    ComponentPose=Local To Component(LocalPose=@n1)
    Alpha=1
    ConstraintWeights_0=0.0141
    ConstraintWeights_1:float=0.5643
    ConstraintWeights_2:float=0.0584
    ConstraintWeights_3:float=0.0338
    ConstraintWeights_4:float=0.1345
    ConstraintWeights_5:float=0.0309
    ConstraintWeights_6:float=0.1641
   ) -> (Pose) {
    alphaBoolBlend {blendOption=Linear}
    boneToModify {boneName=attach_cape}
    constraintSetup [
     {targetBone {boneName=spine_03}}
     {targetBone {boneName=spine_05}}
     {targetBone {boneName=spine_04}}
     {targetBone {boneName=upperarm_twist_01_r}}
     {targetBone {boneName=clavicle_r}}
     {targetBone {boneName=upperarm_twist_01_l}}
     {targetBone {boneName=clavicle_l}}
    ]
    constraintWeights [1,1,1,1,1,1,1]
   }
   n3 = "Constraint - attach_backpack (PPPPPPPP)"!(
    ComponentPose=@n2
    Alpha=1
    ConstraintWeights_0=0.0755
    ConstraintWeights_1:float=0.2367
    ConstraintWeights_2:float=0.2449
    ConstraintWeights_3:float=0.0349
    ConstraintWeights_4:float=0.1732
    ConstraintWeights_5:float=0.0207
    ConstraintWeights_6:float=0.0371
    ConstraintWeights_7:float=0.1771
   ) -> (Pose) {
    alphaBoolBlend {blendOption=Linear}
    boneToModify {boneName=attach_backpack}
    constraintSetup [
     {targetBone {boneName=spine_03}}
     {targetBone {boneName=spine_05}}
     {targetBone {boneName=spine_04}}
     {targetBone {boneName=upperarm_twist_01_r}}
     {targetBone {boneName=clavicle_r}}
     {targetBone {boneName=spine_02}}
     {targetBone {boneName=upperarm_twist_01_l}}
     {targetBone {boneName=clavicle_l}}
    ]
    constraintWeights [1,1,1,1,1,1,1,1]
   }
   n4 = "Transform (Modify) Bone - Bone: dyn_neck_01_constraint"(ComponentPose=@n3, Alpha=Get neck01 back() -> (neck01 back:double), Translation="0,0,0", Rotation="0,-35,0", Scale="1,1,1.1") {
    rotationMode=BMM_Additive
    rotationSpace=BCS_BoneSpace
    scaleMode=BMM_Additive
    scaleSpace=BCS_BoneSpace
    alphaBoolBlend {blendOption=Linear}
    boneToModify {boneName=dyn_neck_01_constraint}
    rotation {yaw=-35}
    scale {z=1.1}
   }
   n5 = "Transform (Modify) Bone - Bone: dyn_neck_02_constraint"(ComponentPose=@n4, Alpha=Get neck02 back() -> (neck02 back:double), Translation="0,0,0", Rotation="0.000,-15.000,-0.000", Scale="1,1,1.2") {
    rotationMode=BMM_Additive
    rotationSpace=BCS_BoneSpace
    scaleMode=BMM_Additive
    scaleSpace=BCS_BoneSpace
    alphaBoolBlend {blendOption=Linear}
    boneToModify {boneName=dyn_neck_02_constraint}
    rotation=<rot 0 -15.0001 0>
    scale {z=1.2}
   }
   n6 = "Constraint - dyn_pouch (PP)"!(ComponentPose=@n5, Alpha=1, ConstraintWeights_0=0.6225, ConstraintWeights_1:float=0.3775) -> (Pose) {
    alphaBoolBlend {blendOption=Linear}
    boneToModify {boneName=dyn_pouch}
    constraintSetup [{targetBone {boneName=spine_01}}, {targetBone {boneName=pelvis}}]
    constraintWeights [1,1]
   }
   n7 = Get R leg up() -> (R leg up:double)
   n8 = "Constraint - dyn_cannister_A (R)"(ComponentPose=@n6, Alpha=@n7, ConstraintWeights_0=1) {
    alphaBoolBlend {blendOption=Linear}
    boneToModify {boneName=dyn_cannister_A}
    constraintSetup [{transformType=Rotation, targetBone {boneName=thigh_r}}]
    constraintWeights [1]
   }
   n9 = "Constraint - dyn_cannister_A (T)"(ComponentPose=@n8, Alpha=@n7, ConstraintWeights_0=1) {
    alphaBoolBlend {blendOption=Linear}
    alphaScaleBiasClamp {scale=0.6}
    boneToModify {boneName=dyn_cannister_A}
    constraintSetup [{transformType=Translation, targetBone {boneName=thigh_r}}]
    constraintWeights [1]
   }
   n10 = "Constraint - dyn_cannister_A (PPPP)"!(
    ComponentPose=@n9
    Alpha=1
    ConstraintWeights_0=0.43
    ConstraintWeights_1:float=0.3187
    ConstraintWeights_2:float=0.2347
    ConstraintWeights_3:float=0.0166
   ) -> (Pose) {
    alphaBoolBlend {blendOption=Linear}
    alphaScaleBiasClamp {scale=0.5}
    boneToModify {boneName=dyn_cannister_A}
    constraintSetup [{targetBone {boneName=thigh_r}}, {targetBone {boneName=pelvis}}, {targetBone {boneName=spine_01}}, {targetBone {boneName=thigh_twist_01_r}}]
    constraintWeights [1,1,1,1]
   }
   n11 = "Constraint - dyn_cannister_B (PPPP)"!(
    ComponentPose=@n10
    Alpha=1
    ConstraintWeights_0=0.4911
    ConstraintWeights_1:float=0.3045
    ConstraintWeights_2:float=0.1729
    ConstraintWeights_3:float=0.0314
   ) -> (Pose) {
    alphaBoolBlend {blendOption=Linear}
    alphaScaleBiasClamp {scale=0.5}
    boneToModify {boneName=dyn_cannister_B}
    constraintSetup [{targetBone {boneName=thigh_r}}, {targetBone {boneName=pelvis}}, {targetBone {boneName=spine_01}}, {targetBone {boneName=thigh_twist_01_r}}]
    constraintWeights [1,1,1,1]
   }
   n12 = "Constraint - dyn_cannister_C (PPPP)"!(
    ComponentPose=@n11
    Alpha=1
    ConstraintWeights_0=0.4143
    ConstraintWeights_1:float=0.3179
    ConstraintWeights_2:float=0.2345
    ConstraintWeights_3:float=0.0333
   ) -> (Pose) {
    alphaBoolBlend {blendOption=Linear}
    alphaScaleBiasClamp {scale=0.5}
    boneToModify {boneName=dyn_cannister_C}
    constraintSetup [{targetBone {boneName=pelvis}}, {targetBone {boneName=thigh_r}}, {targetBone {boneName=spine_01}}, {targetBone {boneName=thigh_twist_01_r}}]
    constraintWeights [1,1,1,1]
   }
   n13 = "Constraint - dyn_cannister_D (PPPPP)"!(
    ComponentPose=@n12
    Alpha=1
    ConstraintWeights_0=0.6932
    ConstraintWeights_1:float=0.1367
    ConstraintWeights_2:float=0.1163
    ConstraintWeights_3:float=0.0306
    ConstraintWeights_4:float=0.0232
   ) -> (Pose) {
    alphaBoolBlend {blendOption=Linear}
    boneToModify {boneName=dyn_cannister_D}
    constraintSetup [
     {targetBone {boneName=pelvis}}
     {targetBone {boneName=thigh_r}}
     {targetBone {boneName=spine_01}}
     {targetBone {boneName=deform_glute_r}}
     {targetBone {boneName=thigh_twist_01_r}}
    ]
    constraintWeights [1,1,1,1,1]
   }
   n14 = "Constraint - dyn_handkerchief (PPPPPPPP)"!(
    ComponentPose=@n13
    Alpha=1
    ConstraintWeights_0=0.0163
    ConstraintWeights_1:float=0.2786
    ConstraintWeights_2:float=0.0374
    ConstraintWeights_3:float=0.0153
    ConstraintWeights_4:float=0.0056
    ConstraintWeights_5:float=0.0697
    ConstraintWeights_6:float=0.1169
    ConstraintWeights_7:float=0.4601
   ) -> (Pose) {
    alphaBoolBlend {blendOption=Linear}
    boneToModify {boneName=dyn_handkerchief}
    constraintSetup [
     {targetBone {boneName=spine_03}}
     {targetBone {boneName=spine_05}}
     {targetBone {boneName=spine_04}}
     {targetBone {boneName=deform_upperarm_up_l}}
     {targetBone {boneName=spine_02}}
     {targetBone {boneName=upperarm_twist_01_l}}
     {targetBone {boneName=deform_upperarm_fr_l}}
     {targetBone {boneName=clavicle_l}}
    ]
    constraintWeights [1,1,1,1,1,1,1,1]
   }
   n15 = Get R arm aimUp() -> (R arm aimUp:double)
   n16 = "Transform (Modify) Bone - Bone: deform_upperarm_bk_r"(ComponentPose=@n14, Alpha=@n15, Translation="0,0,0.5", Rotation="15,0,25", Scale="1,1,1") {
    rotationMode=BMM_Additive
    rotationSpace=BCS_BoneSpace
    translationMode=BMM_Additive
    translationSpace=BCS_BoneSpace
    alphaBoolBlend {blendOption=Linear}
    alphaScaleBiasClamp {bInterpResult=true, interpSpeedDecreasing=20, interpSpeedIncreasing=20}
    boneToModify {boneName=deform_upperarm_bk_r}
    rotation {pitch=15, roll=25}
    translation {z=0.5}
   }
   n17 = "Transform (Modify) Bone - Bone: deform_upperarm_fr_r"(ComponentPose=@n16, Alpha=@n15, Translation="0,0,0", Rotation="0.000,0.000,30.000", Scale="1,1,1") {
    rotationMode=BMM_Additive
    rotationSpace=BCS_BoneSpace
    alphaBoolBlend {blendOption=Linear}
    alphaScaleBiasClamp {bInterpResult=true, interpSpeedDecreasing=20, interpSpeedIncreasing=20}
    boneToModify {boneName=deform_upperarm_fr_r}
    rotation=<rot 0.0001 0 30>
   }
   n18 = Look At - Bone: dyn_cannister_A(ComponentPose=@n17, Alpha=Get R leg up()) {
    alphaBoolBlend {blendOption=Linear}
    boneToModify {boneName=dyn_cannister_A}
    lookAtLocation=<24.4714 -7.0802 -1.1517>
    lookAtTarget {boneReference {boneName=thigh_r}}
    lookAt_Axis {axis=<1 0>}
   }
   n19 = Look At - Bone: dyn_cannister_B(ComponentPose=@n18, Alpha=Get R_legUpPitch avg() -> (R_legUpPitch avg:float)) {
    alphaBoolBlend {blendOption=Linear}
    alphaScaleBiasClamp {scale=0.8}
    boneToModify {boneName=dyn_cannister_B}
    lookAtLocation=<9.9972 -4.8027 5.5805>
    lookAtTarget {boneReference {boneName=thigh_r}}
    lookAt_Axis {axis=<1 0>}
   }
   n20 = Look At - Bone: dyn_cannister_C(ComponentPose=@n19, Alpha=Get R leg pitch() -> (R leg pitch:double)) {
    alphaBoolBlend {blendOption=Linear}
    alphaScaleBiasClamp {scale=0.4}
    boneToModify {boneName=dyn_cannister_C}
    lookAtLocation=<31.5868 -0.6046 5.6082>
    lookAtTarget {boneReference {boneName=thigh_r}}
    lookAt_Axis {axis=<1 0>}
   }
   n21 = Look At - Bone: dyn_cannister_D(ComponentPose=@n20, Alpha=1) {
    alphaBoolBlend {blendOption=Linear}
    alphaScaleBiasClamp {scale=0.2}
    boneToModify {boneName=dyn_cannister_D}
    lookAtLocation=<10.0024 5.6457 6.4941>
    lookAtTarget {boneReference {boneName=thigh_r}}
    lookAt_Axis {axis=<1 0>}
   }
   n22 = "Transform (Modify) Bone - Bone: dyn_cannister_A"(ComponentPose=@n21, Alpha=Get R leg up(), Translation="0.5,-2.4,-3.5", Rotation="-20,0,0", Scale="1,1,1") {
    rotationMode=BMM_Additive
    rotationSpace=BCS_BoneSpace
    translationMode=BMM_Additive
    translationSpace=BCS_BoneSpace
    alphaBoolBlend {blendOption=Linear}
    boneToModify {boneName=dyn_cannister_A}
    rotation {pitch=-20}
    translation=<0.5 -2.4 -3.5>
   }
   n23 = "Transform (Modify) Bone - Bone: dyn_cannister_B"(ComponentPose=@n22, Alpha=Get R_legUpPitch avg(), Translation="2,-1,-2", Rotation="-35,0,0", Scale="1,1,1") {
    rotationMode=BMM_Additive
    rotationSpace=BCS_BoneSpace
    translationMode=BMM_Additive
    translationSpace=BCS_BoneSpace
    alphaBoolBlend {blendOption=Linear}
    boneToModify {boneName=dyn_cannister_B}
    rotation {pitch=-35}
    translation=<2 -1 -2>
   }
   n24 = "Transform (Modify) Bone - Bone: dyn_cannister_C"(ComponentPose=@n23, Alpha=Get R leg pitch(), Translation="0,1,-1", Rotation="-30,-5,0", Scale="1,1,1") {
    rotationMode=BMM_Additive
    rotationSpace=BCS_BoneSpace
    translationMode=BMM_Additive
    translationSpace=BCS_BoneSpace
    alphaBoolBlend {blendOption=Linear}
    boneToModify {boneName=dyn_cannister_C}
    rotation {pitch=-30, yaw=-5}
    translation {y=1, z=-1}
   }
   n25 = Anim Dynamics - Bone: dyn_L_braceletMaster(ComponentPose=@n24, Alpha=1, ExternalForce="0,0,0") {
    angularDampingOverride=1
    bLinearSpring=true
    bOverrideAngularDamping=true
    bUseGravityOverride=true
    gravityScale=0.35
    linearSpringConstant=100
    simulationSpace=RootRelative
    alphaBoolBlend {blendOption=Linear}
    boundBone {boneName=dyn_L_braceletMaster}
    componentLinearAccScale=<1 1 1>
    componentLinearVelScale=<1 1 1>
    gravityOverride {z=-980}
    physicsBodyDefinitions [{
      boundBone {boneName=dyn_L_braceletMaster}
      boxExtents=<60 60 60>
      constraintSetup {angularLimitsMax=<0 0 0>, angularLimitsMin=<0 0 0>, angularTarget=<1 0 0>, linearAxesMin=<-3 0 0>}
      localJointOffset=<1 0 0>
     }]
    relativeSpaceBone {boneName=hand_l}
   }
   n26 = Anim Dynamics - Bone: dyn_L_braceletSecondary(ComponentPose=@n25, Alpha=1, ExternalForce="0,0,0") {
    angularDampingOverride=1
    angularSpringConstant=8000
    bAngularSpring=true
    bOverrideAngularDamping=true
    bUseGravityOverride=true
    gravityScale=0.35
    simulationSpace=RootRelative
    alphaBoolBlend {blendOption=Linear}
    boundBone {boneName=dyn_L_braceletSecondary}
    componentLinearAccScale=<1 1 1>
    componentLinearVelScale=<1 1 1>
    gravityOverride {z=-980}
    physicsBodyDefinitions [{
      boundBone {boneName=dyn_L_braceletSecondary}
      boxExtents=<30 30 30>
      constraintSetup {angularLimitsMax=<0 10 0>, angularLimitsMin=<0 -40 0>, angularTarget=<1 0 0>}
      localJointOffset=<1 0 0>
     }]
    relativeSpaceBone {boneName=hand_l}
   }
   n27 = Anim Dynamics - Bone: dyn_R_braceletMaster(ComponentPose=@n26, Alpha=1, ExternalForce="0,0,0") {
    angularDampingOverride=1
    bLinearSpring=true
    bOverrideAngularDamping=true
    bUseGravityOverride=true
    gravityScale=0.35
    linearSpringConstant=100
    simulationSpace=RootRelative
    alphaBoolBlend {blendOption=Linear}
    boundBone {boneName=dyn_R_braceletMaster}
    componentLinearAccScale=<1 1 1>
    componentLinearVelScale=<1 1 1>
    gravityOverride {z=-980}
    physicsBodyDefinitions [{
      boundBone {boneName=dyn_R_braceletMaster}
      boxExtents=<60 60 60>
      constraintSetup {angularLimitsMax=<0 0 0>, angularLimitsMin=<0 0 0>, angularTarget=<1 0 0>, linearAxesMin=<-3 0 0>}
      localJointOffset=<1 0 0>
     }]
    relativeSpaceBone {boneName=hand_r}
   }
   n28 = Anim Dynamics - Bone: dyn_R_braceletSecondary(ComponentPose=@n27, Alpha=1, ExternalForce="0,0,0") {
    angularDampingOverride=1
    angularSpringConstant=8000
    bAngularSpring=true
    bOverrideAngularDamping=true
    bUseGravityOverride=true
    gravityScale=0.35
    simulationSpace=RootRelative
    alphaBoolBlend {blendOption=Linear}
    boundBone {boneName=dyn_R_braceletSecondary}
    componentLinearAccScale=<1 1 1>
    componentLinearVelScale=<1 1 1>
    gravityOverride {z=-980}
    physicsBodyDefinitions [{
      boundBone {boneName=dyn_R_braceletSecondary}
      boxExtents=<30 30 30>
      constraintSetup {angularLimitsMax=<0 10 0>, angularLimitsMin=<0 -40 0>, angularTarget=<1 0 0>}
      localJointOffset=<1 0 0>
     }]
    relativeSpaceBone {boneName=hand_r}
   }
   n29 = Anim Dynamics - Bone: dyn_L_bootStrap(ComponentPose=@n28, Alpha=1, ExternalForce="0,0,0") {
    angularDampingOverride=0.999
    angularSpringConstant=500
    bAngularSpring=true
    bOverrideAngularDamping=true
    bUseGravityOverride=true
    gravityScale=0.35
    simulationSpace=RootRelative
    alphaBoolBlend {blendOption=Linear}
    boundBone {boneName=dyn_L_bootStrap}
    componentLinearAccScale=<1 1 1>
    componentLinearVelScale=<1 1 1>
    gravityOverride {z=-980}
    physicsBodyDefinitions [{
      boundBone {boneName=dyn_L_bootStrap}
      boxExtents=<10 10 10>
      constraintSetup {angularLimitsMax=<0 0 20>, angularLimitsMin=<0 0 -20>, angularTarget=<1 0 0>}
      localJointOffset=<1 0 0>
     }]
   }
   n30 = Anim Dynamics - Bone: dyn_R_bootStrap(ComponentPose=@n29, Alpha=1, ExternalForce="0,0,0") {
    angularDampingOverride=0.999
    angularSpringConstant=500
    bAngularSpring=true
    bOverrideAngularDamping=true
    bUseGravityOverride=true
    gravityScale=0.35
    simulationSpace=RootRelative
    alphaBoolBlend {blendOption=Linear}
    boundBone {boneName=dyn_R_bootStrap}
    componentLinearAccScale=<1 1 1>
    componentLinearVelScale=<1 1 1>
    gravityOverride {z=-980}
    physicsBodyDefinitions [{
      boundBone {boneName=dyn_R_bootStrap}
      boxExtents=<10 10 10>
      constraintSetup {angularLimitsMax=<0 0 20>, angularLimitsMin=<0 0 -20>, angularTarget=<1 0 0>}
      localJointOffset=<1 0 0>
     }]
   }
   n31 = Anim Dynamics - Bone: dyn_pouch(ComponentPose=@n30, Alpha=1, ExternalForce="0,0,0") {
    angularDampingOverride=0.9999
    angularSpringConstant=15000
    bAngularSpring=true
    bOverrideAngularDamping=true
    bUseGravityOverride=true
    gravityScale=0.35
    simulationSpace=RootRelative
    alphaBoolBlend {blendOption=Linear}
    alphaScaleBiasClamp {scale=0.7}
    boundBone {boneName=dyn_pouch}
    componentLinearAccScale=<1 1 1>
    componentLinearVelScale=<1 1 1>
    gravityOverride {z=-2000}
    physicsBodyDefinitions [{
      boundBone {boneName=dyn_pouch}
      boxExtents=<30 20 20>
      constraintSetup {angularLimitsMax=<10 10 5>, angularLimitsMin=<-10 -10 -30>, angularTarget=<1 0 0>}
      localJointOffset=<1 0 0>
     }]
   }
   n32 = Anim Dynamics - Bone: dyn_handkerchief(ComponentPose=@n31, Alpha=1, ExternalForce="0,0,0") {
    angularDampingOverride=1
    angularSpringConstant=2000
    bAngularSpring=true
    bOverrideAngularDamping=true
    bUseGravityOverride=true
    gravityScale=0.35
    simulationSpace=World
    alphaBoolBlend {blendOption=Linear}
    boundBone {boneName=dyn_handkerchief}
    componentLinearAccScale=<1 1 1>
    componentLinearVelScale=<1 1 1>
    gravityOverride {z=-980}
    physicsBodyDefinitions [{
      boundBone {boneName=dyn_handkerchief}
      boxExtents=<9 9 9>
      constraintSetup {angularLimitsMax=<5 15 55>, angularLimitsMin=<-5 -15 -5>, angularTarget=<1 0 0>}
      localJointOffset=<0 0.5 0>
     }]
   }
   n33 = Anim Dynamics - Bone: dyn_L_collar_A(ComponentPose=@n32, Alpha=1, ExternalForce="0,0,0") {
    angularDampingOverride=1
    bLinearSpring=true
    bOverrideAngularDamping=true
    bOverrideLinearDamping=true
    bUseGravityOverride=true
    gravityScale=0.35
    linearDampingOverride=1
    linearSpringConstant=200
    simulationSpace=BoneRelative
    alphaBoolBlend {blendOption=Linear}
    alphaScaleBiasClamp {scale=0.2}
    boundBone {boneName=dyn_L_collar_A}
    componentLinearAccScale=<0.7 0.7 0.7>
    componentLinearVelScale=<0.7 0.7 0.7>
    gravityOverride {z=-980}
    physicsBodyDefinitions [{
      boundBone {boneName=dyn_L_collar_A}
      boxExtents=<7 7 7>
      constraintSetup {angularLimitsMax=<0 0 0>, angularLimitsMin=<0 0 0>, angularTarget=<1 0 0>, linearAxesMax=<0 0 2>}
      localJointOffset=<0 0 1>
     }]
    relativeSpaceBone {boneName=pelvis}
   }
   n34 = Anim Dynamics - Bone: dyn_R_collar_A(ComponentPose=@n33, Alpha=1, ExternalForce="0,0,0") {
    angularDampingOverride=1
    bLinearSpring=true
    bOverrideAngularDamping=true
    bOverrideLinearDamping=true
    bUseGravityOverride=true
    gravityScale=0.35
    linearDampingOverride=1
    linearSpringConstant=200
    simulationSpace=BoneRelative
    alphaBoolBlend {blendOption=Linear}
    alphaScaleBiasClamp {scale=0.2}
    boundBone {boneName=dyn_R_collar_A}
    componentLinearAccScale=<0.7 0.7 0.7>
    componentLinearVelScale=<0.7 0.7 0.7>
    gravityOverride {z=-980}
    physicsBodyDefinitions [{
      boundBone {boneName=dyn_R_collar_A}
      boxExtents=<7 7 7>
      constraintSetup {angularLimitsMax=<0 0 0>, angularLimitsMin=<0 0 0>, angularTarget=<1 0 0>, linearAxesMin=<0 0 -2>}
      localJointOffset=<0 0 -1>
     }]
    relativeSpaceBone {boneName=pelvis}
   }
   n35 = Anim Dynamics - Bone: dyn_cannister_A(ComponentPose=@n34, Alpha=1, ExternalForce="0,0,0") {
    angularDampingOverride=0.9999
    angularSpringConstant=2000
    bAngularSpring=true
    bOverrideAngularDamping=true
    bUseGravityOverride=true
    gravityScale=0.35
    simulationSpace=BoneRelative
    alphaBoolBlend {blendOption=Linear}
    boundBone {boneName=dyn_cannister_A}
    componentLinearAccScale=<1 1 1>
    componentLinearVelScale=<1 1 1>
    gravityOverride {z=-980}
    physicsBodyDefinitions [{
      boundBone {boneName=dyn_cannister_A}
      boxExtents=<10 10 10>
      constraintSetup {angularLimitsMax=<6 6 6>, angularLimitsMin=<-6 -6 -6>, angularTarget=<1 0 0>}
      localJointOffset=<1 0 0>
     }]
    relativeSpaceBone {boneName=pelvis}
   }
   n36 = Anim Dynamics - Bone: dyn_cannister_B(ComponentPose=@n35, Alpha=1, ExternalForce="0,0,0") {
    angularDampingOverride=0.9999
    angularSpringConstant=2000
    bAngularSpring=true
    bOverrideAngularDamping=true
    bUseGravityOverride=true
    gravityScale=0.35
    simulationSpace=BoneRelative
    alphaBoolBlend {blendOption=Linear}
    boundBone {boneName=dyn_cannister_B}
    componentLinearAccScale=<1 1 1>
    componentLinearVelScale=<1 1 1>
    gravityOverride {z=-980}
    physicsBodyDefinitions [{
      boundBone {boneName=dyn_cannister_B}
      boxExtents=<12 12 12>
      constraintSetup {angularLimitsMax=<6 6 6>, angularLimitsMin=<-6 -6 -6>, angularTarget=<1 0 0>}
      localJointOffset=<1 0 0>
     }]
    relativeSpaceBone {boneName=pelvis}
   }
   n37 = Anim Dynamics - Bone: dyn_cannister_C(ComponentPose=@n36, Alpha=1, ExternalForce="0,0,0") {
    angularDampingOverride=0.9999
    angularSpringConstant=2000
    bAngularSpring=true
    bOverrideAngularDamping=true
    bUseGravityOverride=true
    gravityScale=0.35
    simulationSpace=BoneRelative
    alphaBoolBlend {blendOption=Linear}
    boundBone {boneName=dyn_cannister_C}
    componentLinearAccScale=<1 1 1>
    componentLinearVelScale=<1 1 1>
    gravityOverride {z=-980}
    physicsBodyDefinitions [{
      boundBone {boneName=dyn_cannister_C}
      boxExtents=<11 11 11>
      constraintSetup {angularLimitsMax=<6 6 6>, angularLimitsMin=<-6 -6 -6>, angularTarget=<1 0 0>}
      localJointOffset=<1 0 0>
     }]
    relativeSpaceBone {boneName=pelvis}
   }
   n38 = Anim Dynamics - Bone: dyn_cannister_D(ComponentPose=@n37, Alpha=1, ExternalForce="0,0,0") {
    angularDampingOverride=0.9999
    angularSpringConstant=2000
    bAngularSpring=true
    bOverrideAngularDamping=true
    bUseGravityOverride=true
    gravityScale=0.35
    simulationSpace=BoneRelative
    alphaBoolBlend {blendOption=Linear}
    boundBone {boneName=dyn_cannister_D}
    componentLinearAccScale=<1 1 1>
    componentLinearVelScale=<1 1 1>
    gravityOverride {z=-980}
    physicsBodyDefinitions [{
      boundBone {boneName=dyn_cannister_D}
      boxExtents=<14 14 14>
      constraintSetup {angularLimitsMax=<6 6 6>, angularLimitsMin=<-6 -6 -6>, angularTarget=<1 0 0>}
      localJointOffset=<1 0 0>
     }]
    relativeSpaceBone {boneName=pelvis}
   }
   n39 = Get L leg up() -> (L leg up:double)
   n40 = "Transform (Modify) Bone - Bone: dyn_L_belt_joint_A"(ComponentPose=@n38, Alpha=@n39, Translation="-1.955,0.65,7", Rotation="0,0,10", Scale="1,1.3,1") {
    rotationMode=BMM_Additive
    rotationSpace=BCS_BoneSpace
    scaleMode=BMM_Additive
    scaleSpace=BCS_BoneSpace
    translationMode=BMM_Additive
    translationSpace=BCS_BoneSpace
    alphaBoolBlend {blendOption=Linear}
    boneToModify {boneName=dyn_L_belt_joint_A}
    rotation {roll=10}
    scale {y=1.3}
    translation=<-1.955 0.65 7>
   }
   n41 = "Transform (Modify) Bone - Bone: dyn_L_belt_joint_B"(ComponentPose=@n40, Alpha=@n39, Translation="0,0.781,8.8", Rotation="0,0,20", Scale="1,1,1") {
    rotationMode=BMM_Additive
    rotationSpace=BCS_BoneSpace
    translationMode=BMM_Additive
    translationSpace=BCS_BoneSpace
    alphaBoolBlend {blendOption=Linear}
    boneToModify {boneName=dyn_L_belt_joint_B}
    rotation {roll=20}
    translation {y=0.781, z=8.8}
   }
   n42 = "Transform (Modify) Bone - Bone: dyn_L_belt_joint_C"(ComponentPose=@n41, Alpha=@n39, Translation="0.5,-1,4", Rotation="-0.000,-0.000,5.000", Scale="1,1,1") {
    rotationMode=BMM_Additive
    rotationSpace=BCS_BoneSpace
    translationMode=BMM_Additive
    translationSpace=BCS_BoneSpace
    alphaBoolBlend {blendOption=Linear}
    boneToModify {boneName=dyn_L_belt_joint_C}
    rotation=<rot -0.0001 0 5>
    translation=<0.5 -1 4>
   }
   n43 = "Transform (Modify) Bone - Bone: dyn_L_belt_joint_A"(ComponentPose=@n42, Alpha=Get spineForward() -> (spineForward:double), Translation="-2,1.5,-3", Rotation="20,0,20", Scale="1,1,1") {
    rotationMode=BMM_Additive
    rotationSpace=BCS_BoneSpace
    scaleMode=BMM_Additive
    scaleSpace=BCS_BoneSpace
    translationMode=BMM_Additive
    translationSpace=BCS_BoneSpace
    alphaBoolBlend {blendOption=Linear}
    boneToModify {boneName=dyn_L_belt_joint_A}
    rotation {pitch=20, roll=20}
    translation=<-2 1.5 -3>
   }
   Output Pose(Result=Component To Local(ComponentPose=@n43)) {layerGroup=None}
  ]}
 EventGraph {nodes [
   Event BlueprintUpdateAnimation()
   n1 = GetOwnerSkeletalMeshComponent()
   n2 = BreakTransform(InTransform=GetDeltaTransformFromRefPose(self=@n1, BoneName=thigh_l, BaseName=pelvis))
   > Set L leg up(L leg up:double=MapRangeClamped(Value=BreakRotator(InRot=@n2.Rotation).Yaw, InRangeA=0, InRangeB=-95)) -> (Output_Get:double)
   n3 = BreakRotator(InRot=BreakTransform(InTransform=GetDeltaTransformFromRefPose(self=@n1, BoneName=thigh_r, BaseName=pelvis)).Rotation)
   > n4 = Set R leg up(R leg up:double=MapRangeClamped(Value=@n3.Yaw, InRangeA=-25, InRangeB=-108)) -> (Output_Get:double)
   > n5 = Set R leg pitch(R leg pitch:double=MapRangeClamped(Value=@n3.Pitch, InRangeA=10, InRangeB=35)) -> (Output_Get:double)
   > Set R_legUpPitch avg(R_legUpPitch avg:float=Weighted Moving Average Float!(CurrentSample:float=@n4, PreviousSample:float=@n5, Weight:float=0.7) -> (ReturnValue:float)) -> (Output_Get:float)
   n6 = BreakTransform(InTransform=GetDeltaTransformFromRefPose(self=@n1, BoneName=neck_01, BaseName=spine_05))
   > Set neck01 back(neck01 back:double=MapRangeClamped(Value=BreakRotator(InRot=@n6.Rotation).Yaw, InRangeA=0, InRangeB=-40)) -> (Output_Get:double)
   n7 = BreakTransform(InTransform=GetDeltaTransformFromRefPose(self=@n1, BoneName=neck_02, BaseName=neck_01))
   > Set neck02 back(neck02 back:double=MapRangeClamped(Value=BreakRotator(InRot=@n7.Rotation).Yaw, InRangeA=0, InRangeB=-5)) -> (Output_Get:double)
   n8 = BreakTransform(InTransform=GetDeltaTransformFromRefPose(self=@n1, BoneName=spine_02, BaseName=pelvis))
   > Set spineForward(spineForward:double=MapRangeClamped(Value=BreakRotator(InRot=@n8.Rotation).Yaw, InRangeA=0, InRangeB=30)) -> (Output_Get:double)
   > Branch(Condition=OR Boolean(A=Get bIsCrouching(), B=Get bIsSkydiveFloating()))
   n9 = BreakTransform(InTransform=GetDeltaTransformFromRefPose(self=@n1, BoneName=upperarm_r, BaseName=clavicle_r))
   > Set R arm aimUp(R arm aimUp:double=MapRangeClamped(Value=BreakRotator(InRot=@n9.Rotation).Yaw, InRangeA=90, InRangeB=120)) -> (Output_Get:double)
   n10 = Event OnEndCrouching()
   Set R arm aimUp(execute=[@n10.then, @n11.then], R arm aimUp=0)
   n11 = Event OnEndSkydiving()
  ]}
]
variables [
 L leg up {type=number}
 neck01 back {type=number}
 neck02 back {type=number}
 R arm aimUp {type=number}
 R leg pitch {type=number}
 R leg up {type=number}
 R_legUpPitch avg {type=number}
 spineForward {type=number}
]
